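# Sound cue for a scan sound in the Battle Pass Season 26 Part 1 quest Q12.
# A Random node picks one of three equally weighted scan wave variations each time it plays.
# Playback routes through the quest sound class with a dedicated attenuation asset for spatial falloff and a slight volume boost.
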SoundCue /BattlePassS26_Quests/Sound/Quest_S26_Pt1_Q12/Quest_S26_Pt1_Q12_Scan_Cue
root=Random_0
nodes [
 {id=Random_0, children [WavePlayer_0,WavePlayer_1,WavePlayer_2], properties {weights [1,1,1]}}
 {id=WavePlayer_0, properties {soundWaveAssetPtr=&/Quest_S26_Pt1_Q12_Scan_03}}
 {id=WavePlayer_1, properties {soundWaveAssetPtr=&/Quest_S26_Pt1_Q12_Scan_01}}
 {id=WavePlayer_2, properties {soundWaveAssetPtr=&/Quest_S26_Pt1_Q12_Scan_02}}
]
properties {
 attenuationSettings=&/S26_Pt01_Q12_Atten
 maxDistance=3050
 soundClassObject=&Fort_Quests
 volumeMultiplier=1.2
}
